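# Physics asset for the Burger militia mascot player body, covering dangling gear: front armor, pouches, grenades and a left dangle chain.
# Pelvis and thighs carry capsules; tiny "_MSTR" boxes act as anchor bodies, linked to the pelvis by free-rotation constraints.
# Pouches and grenades hang from their anchors with swing and twist limits, some as two-link chains, while the armor chain rotates freely.
# The dangle chain hangs off the front-left pouch.
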PhysicsAsset /BRCosmetics/Characters/Player/Male/Medium/Bodies/M_MED_Mascot_Militia_Burger/Meshes/M_MED_Mascot_Militia_Burger_Physics
bodies {
 dyn_armor_front_01 [{shapeType=Box, extentX=0.01, extentY=0.01, extentZ=0.01}]
 dyn_armor_front_01_MSTR [{shapeType=Box, extentX=0.01, extentY=0.01, extentZ=0.01}]
 dyn_armor_front_02 [{shapeType=Box, extentX=27.0259, extentY=21.7652, extentZ=3.4677, center=<8.4731 0 -0.7817>, rotation=<rot -14.9999 0 0>}]
 dyn_dangle_l_01 [{shapeType=Box, extentX=4, extentY=5.5, extentZ=1.5, center=<2.1755 0.0001 0.0001>}]
 dyn_dangle_l_02 [{shapeType=Box, extentX=4, extentY=5.5, extentZ=1.5, center=<2.1755 0.0001 0.0001>}]
 dyn_grenade_l_01 [{shapeType=Capsule, length=10.35, radius=3.85, center=<7.6716 -0.1083 0.1869>, rotation=<rot 90 0.0132 0.013>}]
 dyn_grenade_l_01_MSTR [{shapeType=Box, extentX=0.01, extentY=0.01, extentZ=0.01}]
 dyn_grenade_r_01 [{shapeType=Capsule, length=11.35, radius=2.85, center=<7.6716 -0.1083 0.1869>, rotation=<rot 90 0.0132 0.013>}]
 dyn_grenade_r_01_MSTR [{shapeType=Box, extentX=0.01, extentY=0.01, extentZ=0.01}]
 dyn_pouchA_l_01 [{shapeType=Box, extentX=8.0477, extentY=7.3506, extentZ=5.1299, center=<1.3648 0 0>}]
 dyn_pouchA_l_01_MSTR [{shapeType=Box, extentX=0.01, extentY=0.01, extentZ=0.01}]
 dyn_pouchA_l_02 [{shapeType=Box, extentX=9.0665, extentY=6.4159, extentZ=6.1268, center=<1.1675 0 0>}]
 dyn_pouchA_r_01 [{shapeType=Box, extentX=8.0477, extentY=7.3506, extentZ=5.1299, center=<1.3648 0 0>}]
 dyn_pouchA_r_01_MSTR [{shapeType=Box, extentX=0.01, extentY=0.01, extentZ=0.01}]
 dyn_pouchA_r_02 [{shapeType=Box, extentX=9.0665, extentY=6.4159, extentZ=6.1268, center=<1.1675 0 0>}]
 dyn_pouchB_l_01 [{shapeType=Box, extentX=8.0477, extentY=7.3506, extentZ=5.1299, center=<1.3648 0 0>}]
 dyn_pouchB_l_01_MSTR [{shapeType=Box, extentX=0.01, extentY=0.01, extentZ=0.01}]
 dyn_pouchB_l_02 [{shapeType=Box, extentX=9.0665, extentY=6.4159, extentZ=6.1268, center=<1.1675 0 0>}]
 dyn_pouchB_r_01 [{shapeType=Box, extentX=8.0477, extentY=7.3506, extentZ=5.1299, center=<1.3648 0 0>}]
 dyn_pouchB_r_01_MSTR [{shapeType=Box, extentX=0.01, extentY=0.01, extentZ=0.01}]
 dyn_pouchB_r_02 [{shapeType=Box, extentX=9.0665, extentY=6.4159, extentZ=6.1268, center=<1.1675 0 0>}]
 dyn_pouch_front_02 [{shapeType=Box, extentX=9.0665, extentY=6.4159, extentZ=6.1268, center=<1.1675 0 0>}]
 dyn_pouch_front_02_MSTR [{shapeType=Box, extentX=0.01, extentY=0.01, extentZ=0.01}]
 dyn_pouch_front_03 [{shapeType=Box, extentX=8.0477, extentY=7.3506, extentZ=5.1299, center=<1.3648 0 0>}]
 dyn_pouch_front_03_MSTR [{shapeType=Box, extentX=0.01, extentY=0.01, extentZ=0.01}]
 dyn_pouch_front_l_02 [{shapeType=Box, extentX=9.0665, extentY=6.4159, extentZ=6.1268, center=<1.1675 0 0>}]
 dyn_pouch_front_l_02_MSTR [{shapeType=Box, extentX=0.01, extentY=0.01, extentZ=0.01}]
 dyn_pouch_front_l_03 [{shapeType=Box, extentX=8.0477, extentY=7.3506, extentZ=5.1299, center=<1.3648 0 0>}]
 dyn_pouch_front_l_03_MSTR [{shapeType=Box, extentX=0.01, extentY=0.01, extentZ=0.01}]
 dyn_pouch_front_r_02 [{shapeType=Box, extentX=9.0665, extentY=6.4159, extentZ=6.1268, center=<1.1675 0 0>}]
 dyn_pouch_front_r_02_MSTR [{shapeType=Box, extentX=0.01, extentY=0.01, extentZ=0.01}]
 dyn_pouch_front_r_03 [{shapeType=Box, extentX=8.0477, extentY=7.3506, extentZ=5.1299, center=<1.3648 0 0>}]
 dyn_pouch_front_r_03_MSTR [{shapeType=Box, extentX=0.01, extentY=0.01, extentZ=0.01}]
 dyn_strap_back_01_l_MSTR [{shapeType=Box, extentX=0.01, extentY=0.01, extentZ=0.01}]
 dyn_strap_back_01_r_MSTR [{shapeType=Box, extentX=0.01, extentY=0.01, extentZ=0.01}]
 pelvis [{shapeType=Capsule, length=17.1965, radius=10.1665}]
 thigh_l [{shapeType=Capsule, length=40.2772, radius=10.1353, center=<-19.9214 0 0>, rotation=<rot 90 -2.5305 -2.5305>}]
 thigh_r [{shapeType=Capsule, length=40.2772, radius=10.1353, center=<19.9214 0 0>, rotation=<rot 90 1.2659 -178.7342>}]
}
constraints [
 {bone1Name=dyn_grenade_r_01, bone2Name=dyn_grenade_r_01_MSTR, swing1LimitDegrees=45, swing2LimitDegrees=45, twistLimitDegrees=5}
 {bone1Name=dyn_grenade_l_01, bone2Name=dyn_grenade_l_01_MSTR, swing1LimitDegrees=45, swing2LimitDegrees=45, twistLimitDegrees=5}
 {bone1Name=dyn_pouch_front_l_03, bone2Name=dyn_pouch_front_l_03_MSTR, swing1LimitDegrees=15, swing2LimitDegrees=15, twistLimitDegrees=5}
 {bone1Name=dyn_pouch_front_l_02, bone2Name=dyn_pouch_front_l_02_MSTR, swing1LimitDegrees=15, swing2LimitDegrees=15, twistLimitDegrees=5}
 {bone1Name=dyn_pouch_front_03, bone2Name=dyn_pouch_front_03_MSTR, swing1LimitDegrees=15, swing2LimitDegrees=15, twistLimitDegrees=5}
 {bone1Name=dyn_pouch_front_r_03, bone2Name=dyn_pouch_front_r_03_MSTR, swing1LimitDegrees=15, swing2LimitDegrees=15, twistLimitDegrees=5}
 {bone1Name=dyn_pouch_front_02, bone2Name=dyn_pouch_front_02_MSTR, swing1LimitDegrees=15, swing2LimitDegrees=15, twistLimitDegrees=5}
 {bone1Name=dyn_pouch_front_r_02, bone2Name=dyn_pouch_front_r_02_MSTR, swing1LimitDegrees=15, swing2LimitDegrees=15, twistLimitDegrees=5}
 {bone1Name=dyn_pouchA_l_01, bone2Name=dyn_pouchA_l_01_MSTR, swing1LimitDegrees=15, swing2LimitDegrees=15, twistLimitDegrees=5}
 {bone1Name=dyn_pouchA_l_02, bone2Name=dyn_pouchA_l_01, swing1LimitDegrees=15, swing2LimitDegrees=15, twistLimitDegrees=5}
 {bone1Name=dyn_pouchB_l_01, bone2Name=dyn_pouchB_l_01_MSTR, swing1LimitDegrees=15, swing2LimitDegrees=15, twistLimitDegrees=5}
 {bone1Name=dyn_pouchB_l_02, bone2Name=dyn_pouchB_l_01, swing1LimitDegrees=15, swing2LimitDegrees=15, twistLimitDegrees=5}
 {bone1Name=dyn_pouchB_r_01, bone2Name=dyn_pouchB_r_01_MSTR, swing1LimitDegrees=15, swing2LimitDegrees=15, twistLimitDegrees=5}
 {bone1Name=dyn_pouchB_r_02, bone2Name=dyn_pouchB_r_01, swing1LimitDegrees=15, swing2LimitDegrees=15, twistLimitDegrees=5}
 {bone1Name=dyn_pouchA_r_01, bone2Name=dyn_pouchA_r_01_MSTR, swing1LimitDegrees=15, swing2LimitDegrees=15, twistLimitDegrees=5}
 {bone1Name=dyn_pouchA_r_02, bone2Name=dyn_pouchA_r_01, swing1LimitDegrees=15, swing2LimitDegrees=15, twistLimitDegrees=5}
 {bone1Name=dyn_dangle_l_02, bone2Name=dyn_pouch_front_l_02, swing1LimitDegrees=45, swing2LimitDegrees=45, twistLimitDegrees=5}
 {bone1Name=dyn_dangle_l_01, bone2Name=dyn_dangle_l_02, swing1LimitDegrees=45, swing2LimitDegrees=45, twistLimitDegrees=5}
 {
  bone1Name=dyn_armor_front_01
  bone2Name=dyn_armor_front_01_MSTR
  swing1LimitDegrees=45
  swing1Motion=Free
  swing2LimitDegrees=45
  swing2Motion=Free
  twistLimitDegrees=45
  twistMotion=Free
 }
 {
  bone1Name=dyn_strap_back_01_l_MSTR
  bone2Name=pelvis
  swing1LimitDegrees=45
  swing1Motion=Free
  swing2LimitDegrees=45
  swing2Motion=Free
  twistLimitDegrees=45
  twistMotion=Free
 }
 {
  bone1Name=dyn_strap_back_01_r_MSTR
  bone2Name=pelvis
  swing1LimitDegrees=45
  swing1Motion=Free
  swing2LimitDegrees=45
  swing2Motion=Free
  twistLimitDegrees=45
  twistMotion=Free
 }
 {
  bone1Name=dyn_grenade_r_01_MSTR
  bone2Name=pelvis
  swing1LimitDegrees=45
  swing1Motion=Free
  swing2LimitDegrees=45
  swing2Motion=Free
  twistLimitDegrees=45
  twistMotion=Free
 }
 {
  bone1Name=dyn_pouch_front_r_03_MSTR
  bone2Name=pelvis
  swing1LimitDegrees=45
  swing1Motion=Free
  swing2LimitDegrees=45
  swing2Motion=Free
  twistLimitDegrees=45
  twistMotion=Free
 }
 {
  bone1Name=dyn_pouch_front_03_MSTR
  bone2Name=pelvis
  swing1LimitDegrees=45
  swing1Motion=Free
  swing2LimitDegrees=45
  swing2Motion=Free
  twistLimitDegrees=45
  twistMotion=Free
 }
 {
  bone1Name=dyn_pouch_front_l_03_MSTR
  bone2Name=pelvis
  swing1LimitDegrees=45
  swing1Motion=Free
  swing2LimitDegrees=45
  swing2Motion=Free
  twistLimitDegrees=45
  twistMotion=Free
 }
 {
  bone1Name=dyn_pouch_front_l_02_MSTR
  bone2Name=pelvis
  swing1LimitDegrees=45
  swing1Motion=Free
  swing2LimitDegrees=45
  swing2Motion=Free
  twistLimitDegrees=45
  twistMotion=Free
 }
 {
  bone1Name=dyn_pouch_front_02_MSTR
  bone2Name=pelvis
  swing1LimitDegrees=45
  swing1Motion=Free
  swing2LimitDegrees=45
  swing2Motion=Free
  twistLimitDegrees=45
  twistMotion=Free
 }
 {
  bone1Name=dyn_pouch_front_r_02_MSTR
  bone2Name=pelvis
  swing1LimitDegrees=45
  swing1Motion=Free
  swing2LimitDegrees=45
  swing2Motion=Free
  twistLimitDegrees=45
  twistMotion=Free
 }
 {
  bone1Name=dyn_pouchA_l_01_MSTR
  bone2Name=pelvis
  swing1LimitDegrees=45
  swing1Motion=Free
  swing2LimitDegrees=45
  swing2Motion=Free
  twistLimitDegrees=45
  twistMotion=Free
 }
 {
  bone1Name=dyn_pouchB_l_01_MSTR
  bone2Name=pelvis
  swing1LimitDegrees=45
  swing1Motion=Free
  swing2LimitDegrees=45
  swing2Motion=Free
  twistLimitDegrees=45
  twistMotion=Free
 }
 {
  bone1Name=dyn_grenade_l_01_MSTR
  bone2Name=pelvis
  swing1LimitDegrees=45
  swing1Motion=Free
  swing2LimitDegrees=45
  swing2Motion=Free
  twistLimitDegrees=45
  twistMotion=Free
 }
 {
  bone1Name=dyn_pouchB_r_01_MSTR
  bone2Name=pelvis
  swing1LimitDegrees=45
  swing1Motion=Free
  swing2LimitDegrees=45
  swing2Motion=Free
  twistLimitDegrees=45
  twistMotion=Free
 }
 {
  bone1Name=dyn_pouchA_r_01_MSTR
  bone2Name=pelvis
  swing1LimitDegrees=45
  swing1Motion=Free
  swing2LimitDegrees=45
  swing2Motion=Free
  twistLimitDegrees=45
  twistMotion=Free
 }
 {
  bone1Name=thigh_l
  bone2Name=pelvis
  swing1LimitDegrees=45
  swing1Motion=Free
  swing2LimitDegrees=45
  swing2Motion=Free
  twistLimitDegrees=45
  twistMotion=Free
 }
 {
  bone1Name=thigh_r
  bone2Name=pelvis
  swing1LimitDegrees=45
  swing1Motion=Free
  swing2LimitDegrees=45
  swing2Motion=Free
  twistLimitDegrees=45
  twistMotion=Free
 }
 {
  bone1Name=dyn_armor_front_01_MSTR
  bone2Name=pelvis
  swing1LimitDegrees=45
  swing1Motion=Free
  swing2LimitDegrees=45
  swing2Motion=Free
  twistLimitDegrees=45
  twistMotion=Free
 }
 {
  bone1Name=dyn_armor_front_02
  bone2Name=dyn_armor_front_01
  swing1LimitDegrees=45
  swing1Motion=Free
  swing2LimitDegrees=45
  swing2Motion=Free
  twistLimitDegrees=45
  twistMotion=Free
 }
]
