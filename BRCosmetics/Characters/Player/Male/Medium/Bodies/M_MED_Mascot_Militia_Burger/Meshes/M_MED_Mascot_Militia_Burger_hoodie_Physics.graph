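# Physics asset for the hoodie mesh of the male medium Militia Burger mascot body, defining physics bodies for its dynamic hoodie bones.
# Box bodies on the front and back hoodie bones each hang from a tiny MSTR anchor box with limited swing and tight twist.
# The anchors, including two strap anchors without child bodies, link to the pelvis with free angular motion.
# Capsules cover the pelvis, neck and head, chained with free angular motion.
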PhysicsAsset /BRCosmetics/Characters/Player/Male/Medium/Bodies/M_MED_Mascot_Militia_Burger/Meshes/M_MED_Mascot_Militia_Burger_hoodie_Physics
bodies {
 dyn_hoodie_01_l [{shapeType=Box, extentX=5.5, extentY=10.5, extentZ=3.5, center=<2.9845 0 -0.0001>}]
 dyn_hoodie_01_l_MSTR [{shapeType=Box, extentX=0.01, extentY=0.01, extentZ=0.01}]
 dyn_hoodie_01_r [{shapeType=Box, extentX=5.5, extentY=10.5, extentZ=3.5, center=<-2.9845 0 0.0001>, rotation=<rot 0 0 180>}]
 dyn_hoodie_01_r_MSTR [{shapeType=Box, extentX=0.01, extentY=0.01, extentZ=0.01}]
 dyn_hoodie_back_01 [{shapeType=Box, extentX=9, extentY=10.5, extentZ=3.5, center=<5.2162 0 0.0001>}]
 dyn_hoodie_back_01_MSTR [{shapeType=Box, extentX=0.01, extentY=0.01, extentZ=0.01}]
 dyn_hoodie_back_01_l [{shapeType=Box, extentX=9.5, extentY=10.5, extentZ=3.5, center=<4.1711 0.0001 0.0001>}]
 dyn_hoodie_back_01_l_MSTR [{shapeType=Box, extentX=0.01, extentY=0.01, extentZ=0.01}]
 dyn_hoodie_back_01_r [{shapeType=Box, extentX=9.5, extentY=10.5, extentZ=3.5, center=<-4.1711 -0.0001 -0.0001>, rotation=<rot 0 0 -180>}]
 dyn_hoodie_back_01_r_MSTR [{shapeType=Box, extentX=0.01, extentY=0.01, extentZ=0.01}]
 dyn_hoodieback_01 [{shapeType=Box, extentX=5.5, extentY=10.5, extentZ=3.5, center=<2.9845 0 -0.0001>}]
 dyn_hoodieback_01_MSTR [{shapeType=Box, extentX=0.01, extentY=0.01, extentZ=0.01}]
 dyn_strap_back_01_l_MSTR [{shapeType=Box, extentX=0.01, extentY=0.01, extentZ=0.01}]
 dyn_strap_back_01_r_MSTR [{shapeType=Box, extentX=0.01, extentY=0.01, extentZ=0.01}]
 head [{shapeType=Capsule, length=3.75, radius=11.75, center=<2.5548 0 0>, rotation=<rot 90 0 0>}]
 neck_01 [{shapeType=Capsule, length=9.5, radius=11.75, rotation=<rot 90 0 0>}]
 neck_02 [{shapeType=Capsule, length=9.5, radius=11.75, rotation=<rot 90 0 0>}]
 pelvis [{shapeType=Capsule, length=17.1965, radius=10.1665}]
}
constraints [
 {bone1Name=dyn_hoodie_01_l, bone2Name=dyn_hoodie_01_l_MSTR, swing1LimitDegrees=45, swing2LimitDegrees=120, twistLimitDegrees=5}
 {bone1Name=dyn_hoodie_back_01, bone2Name=dyn_hoodie_back_01_MSTR, swing1LimitDegrees=45, swing2LimitDegrees=90, twistLimitDegrees=5}
 {bone1Name=dyn_hoodieback_01, bone2Name=dyn_hoodieback_01_MSTR, swing1LimitDegrees=15, swing2LimitDegrees=45, twistLimitDegrees=5}
 {bone1Name=dyn_hoodie_01_r, bone2Name=dyn_hoodie_01_r_MSTR, swing1LimitDegrees=45, swing2LimitDegrees=120, twistLimitDegrees=5}
 {bone1Name=dyn_hoodie_back_01_r, bone2Name=dyn_hoodie_back_01_r_MSTR, swing1LimitDegrees=45, swing2LimitDegrees=120, twistLimitDegrees=5}
 {bone1Name=dyn_hoodie_back_01_l, bone2Name=dyn_hoodie_back_01_l_MSTR, swing1LimitDegrees=45, swing2LimitDegrees=120, twistLimitDegrees=5}
 {
  bone1Name=neck_02
  bone2Name=neck_01
  swing1LimitDegrees=45
  swing1Motion=Free
  swing2LimitDegrees=45
  swing2Motion=Free
  twistLimitDegrees=45
  twistMotion=Free
 }
 {
  bone1Name=head
  bone2Name=neck_02
  swing1LimitDegrees=45
  swing1Motion=Free
  swing2LimitDegrees=45
  swing2Motion=Free
  twistLimitDegrees=45
  twistMotion=Free
 }
 {
  bone1Name=dyn_strap_back_01_l_MSTR
  bone2Name=pelvis
  swing1LimitDegrees=45
  swing1Motion=Free
  swing2LimitDegrees=45
  swing2Motion=Free
  twistLimitDegrees=45
  twistMotion=Free
 }
 {
  bone1Name=dyn_strap_back_01_r_MSTR
  bone2Name=pelvis
  swing1LimitDegrees=45
  swing1Motion=Free
  swing2LimitDegrees=45
  swing2Motion=Free
  twistLimitDegrees=45
  twistMotion=Free
 }
 {
  bone1Name=neck_01
  bone2Name=pelvis
  swing1LimitDegrees=45
  swing1Motion=Free
  swing2LimitDegrees=45
  swing2Motion=Free
  twistLimitDegrees=45
  twistMotion=Free
 }
 {
  bone1Name=dyn_hoodie_01_l_MSTR
  bone2Name=pelvis
  swing1LimitDegrees=45
  swing1Motion=Free
  swing2LimitDegrees=45
  swing2Motion=Free
  twistLimitDegrees=45
  twistMotion=Free
 }
 {
  bone1Name=dyn_hoodie_back_01_l_MSTR
  bone2Name=pelvis
  swing1LimitDegrees=45
  swing1Motion=Free
  swing2LimitDegrees=45
  swing2Motion=Free
  twistLimitDegrees=45
  twistMotion=Free
 }
 {
  bone1Name=dyn_hoodie_back_01_MSTR
  bone2Name=pelvis
  swing1LimitDegrees=45
  swing1Motion=Free
  swing2LimitDegrees=45
  swing2Motion=Free
  twistLimitDegrees=45
  twistMotion=Free
 }
 {
  bone1Name=dyn_hoodie_back_01_r_MSTR
  bone2Name=pelvis
  swing1LimitDegrees=45
  swing1Motion=Free
  swing2LimitDegrees=45
  swing2Motion=Free
  twistLimitDegrees=45
  twistMotion=Free
 }
 {
  bone1Name=dyn_hoodie_01_r_MSTR
  bone2Name=pelvis
  swing1LimitDegrees=45
  swing1Motion=Free
  swing2LimitDegrees=45
  swing2Motion=Free
  twistLimitDegrees=45
  twistMotion=Free
 }
 {
  bone1Name=dyn_hoodieback_01_MSTR
  bone2Name=pelvis
  swing1LimitDegrees=45
  swing1Motion=Free
  swing2LimitDegrees=45
  swing2Motion=Free
  twistLimitDegrees=45
  twistMotion=Free
 }
]
