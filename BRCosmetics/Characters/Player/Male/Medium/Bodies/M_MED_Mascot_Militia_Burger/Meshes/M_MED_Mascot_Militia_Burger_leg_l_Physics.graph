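# Physics asset for the left leg mesh of the Militia Burger mascot character body, setting up secondary motion for a leg dangle and a shoe strap.
# Each of those two box bodies hinges on a tiny master body, with swing mostly on one axis, wider for the dangle.
# Both masters, plus a shoe master with no child body, connect to the thigh capsule through angularly free constraints.
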PhysicsAsset /BRCosmetics/Characters/Player/Male/Medium/Bodies/M_MED_Mascot_Militia_Burger/Meshes/M_MED_Mascot_Militia_Burger_leg_l_Physics
bodies {
 dyn_leg_dangle_01_l [{shapeType=Box, extentX=4.3876, extentY=1.8083, extentZ=7.656, center=<2.6805 0 0>, rotation=<rot 0 0 89.9991>}]
 dyn_leg_dangle_01_l_MSTR [{shapeType=Box, extentX=0.01, extentY=0.01, extentZ=0.01}]
 dyn_shoe_01_l_MSTR [{shapeType=Box, extentX=0.01, extentY=0.01, extentZ=0.01}]
 dyn_shoestrap_01_l [{shapeType=Box, extentX=4.3876, extentY=1.8083, extentZ=4.656, center=<2.6805 0 0>, rotation=<rot 0 0 89.9991>}]
 dyn_shoestrap_01_l_MSTR [{shapeType=Box, extentX=0.01, extentY=0.01, extentZ=0.01}]
 thigh_l [{shapeType=Capsule, length=40.2772, radius=10.1353, center=<-19.9214 0 0>, rotation=<rot 90 -2.5305 -2.5305>}]
}
constraints [
 {bone1Name=dyn_leg_dangle_01_l, bone2Name=dyn_leg_dangle_01_l_MSTR, swing1LimitDegrees=1, swing2LimitDegrees=50, twistLimitDegrees=1}
 {bone1Name=dyn_shoestrap_01_l, bone2Name=dyn_shoestrap_01_l_MSTR, swing1LimitDegrees=1, swing2LimitDegrees=25, twistLimitDegrees=1}
 {
  bone1Name=dyn_leg_dangle_01_l_MSTR
  bone2Name=thigh_l
  swing1LimitDegrees=45
  swing1Motion=Free
  swing2LimitDegrees=45
  swing2Motion=Free
  twistLimitDegrees=45
  twistMotion=Free
 }
 {
  bone1Name=dyn_shoestrap_01_l_MSTR
  bone2Name=thigh_l
  swing1LimitDegrees=45
  swing1Motion=Free
  swing2LimitDegrees=45
  swing2Motion=Free
  twistLimitDegrees=45
  twistMotion=Free
 }
 {
  bone1Name=dyn_shoe_01_l_MSTR
  bone2Name=thigh_l
  swing1LimitDegrees=45
  swing1Motion=Free
  swing2LimitDegrees=45
  swing2Motion=Free
  twistLimitDegrees=45
  twistMotion=Free
 }
]
