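# Physics asset for the right leg mesh of the Militia Burger mascot player body, covering its dangling leg piece and shoestrap.
# Box bodies on the dangle and strap bones are constrained to tiny MSTR anchor bodies, with limits that favor one swing axis for a near-hinge motion.
# The dangle, shoestrap and shoe anchors link to a thigh capsule with free angular motion.
# The shoe anchor has no child body.
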PhysicsAsset /BRCosmetics/Characters/Player/Male/Medium/Bodies/M_MED_Mascot_Militia_Burger/Meshes/M_MED_Mascot_Militia_Burger_leg_r_Physics
bodies {
 dyn_leg_dangle_01_r [{shapeType=Box, extentX=4.3876, extentY=1.8083, extentZ=7.656, center=<-2.6805 0 0>, rotation=<rot 0 0 -90.0009>}]
 dyn_leg_dangle_01_r_MSTR [{shapeType=Box, extentX=0.01, extentY=0.01, extentZ=0.01}]
 dyn_shoe_01_r_MSTR [{shapeType=Box, extentX=0.01, extentY=0.01, extentZ=0.01}]
 dyn_shoestrap_01_r [{shapeType=Box, extentX=4.3876, extentY=1.8083, extentZ=4.656, center=<-2.6805 0 0>, rotation=<rot 0 0 -90.0009>}]
 dyn_shoestrap_01_r_MSTR [{shapeType=Box, extentX=0.01, extentY=0.01, extentZ=0.01}]
 thigh_r [{shapeType=Capsule, length=40.2772, radius=10.1353, center=<19.9214 0 0>, rotation=<rot 90 1.2659 -178.7342>}]
}
constraints [
 {bone1Name=dyn_leg_dangle_01_r, bone2Name=dyn_leg_dangle_01_r_MSTR, swing1LimitDegrees=2, swing2LimitDegrees=40, twistLimitDegrees=2}
 {bone1Name=dyn_shoestrap_01_r, bone2Name=dyn_shoestrap_01_r_MSTR, swing1LimitDegrees=1, swing2LimitDegrees=25, twistLimitDegrees=1}
 {
  bone1Name=dyn_leg_dangle_01_r_MSTR
  bone2Name=thigh_r
  swing1LimitDegrees=45
  swing1Motion=Free
  swing2LimitDegrees=45
  swing2Motion=Free
  twistLimitDegrees=45
  twistMotion=Free
 }
 {
  bone1Name=dyn_shoestrap_01_r_MSTR
  bone2Name=thigh_r
  swing1LimitDegrees=45
  swing1Motion=Free
  swing2LimitDegrees=45
  swing2Motion=Free
  twistLimitDegrees=45
  twistMotion=Free
 }
 {
  bone1Name=dyn_shoe_01_r_MSTR
  bone2Name=thigh_r
  swing1LimitDegrees=45
  swing1Motion=Free
  swing2LimitDegrees=45
  swing2Motion=Free
  twistLimitDegrees=45
  twistMotion=Free
 }
]
