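# Physics asset for the front loincloth flap on the medium male Militia Burger mascot player cosmetic body.
# A box body on dyn_armor_front_01 hinges on a tiny MSTR anchor box with swing and twist limit angles set, shaping how the flap swings.
# The anchor and two thigh capsules each join the pelvis capsule with free angular motion.
# The pelvis and thigh capsules approximate the lower body, apparently as collision for the flap.
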PhysicsAsset /BRCosmetics/Characters/Player/Male/Medium/Bodies/M_MED_Mascot_Militia_Burger/Meshes/M_MED_Mascot_Militia_Burger_loincloth_Physics
bodies {
 dyn_armor_front_01 [{shapeType=Box, extentX=18, extentY=16, extentZ=4, center=<12.0158 0.0001 -0.09>}]
 dyn_armor_front_01_MSTR [{shapeType=Box, extentX=0.01, extentY=0.01, extentZ=0.01}]
 pelvis [{shapeType=Capsule, length=17.1965, radius=10.1665}]
 thigh_l [{shapeType=Capsule, length=51.5272, radius=10.1353, center=<-19.9214 0 0>, rotation=<rot 90 -2.5305 -2.5305>}]
 thigh_r [{shapeType=Capsule, length=51.5272, radius=10.1353, center=<19.9214 0 0>, rotation=<rot 90 1.2659 -178.7342>}]
}
constraints [
 {bone1Name=dyn_armor_front_01, bone2Name=dyn_armor_front_01_MSTR, swing1LimitDegrees=65, swing2LimitDegrees=90, twistLimitDegrees=25}
 {
  bone1Name=thigh_l
  bone2Name=pelvis
  swing1LimitDegrees=45
  swing1Motion=Free
  swing2LimitDegrees=45
  swing2Motion=Free
  twistLimitDegrees=45
  twistMotion=Free
 }
 {
  bone1Name=dyn_armor_front_01_MSTR
  bone2Name=pelvis
  swing1LimitDegrees=45
  swing1Motion=Free
  swing2LimitDegrees=45
  swing2Motion=Free
  twistLimitDegrees=45
  twistMotion=Free
 }
 {
  bone1Name=thigh_r
  bone2Name=pelvis
  swing1LimitDegrees=45
  swing1Motion=Free
  swing2LimitDegrees=45
  swing2Motion=Free
  twistLimitDegrees=45
  twistMotion=Free
 }
]
