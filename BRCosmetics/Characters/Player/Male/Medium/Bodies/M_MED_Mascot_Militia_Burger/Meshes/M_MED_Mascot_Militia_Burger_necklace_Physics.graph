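# Physics asset for the Militia Burger mascot necklace, simulating a hanging sign as left and right six-segment bone chains.
# Each chain hangs from spine_05 through a tiny MSTR anchor body, with angular limits and tight twist along its segments.
# Free-rotation constraints tie the two chains together at segments 04 to 06.
# Pelvis, spine, neck and head capsules, plus two large slab boxes on spine_03, approximate the mascot body.
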PhysicsAsset /BRCosmetics/Characters/Player/Male/Medium/Bodies/M_MED_Mascot_Militia_Burger/Meshes/M_MED_Mascot_Militia_Burger_necklace_Physics
bodies {
 dyn_sign_01_l [{shapeType=Box, extentX=4, extentY=2, extentZ=1.5, center=<2.1755 0.0001 0.0001>}]
 dyn_sign_01_l_MSTR [{shapeType=Box, extentX=0.01, extentY=0.01, extentZ=0.01}]
 dyn_sign_01_r [{shapeType=Box, extentX=4, extentY=2, extentZ=1.5, center=<-2.1755 -0.0001 -0.0001>, rotation=<rot 0 0 -180>}]
 dyn_sign_01_r_MSTR [{shapeType=Box, extentX=0.01, extentY=0.01, extentZ=0.01}]
 dyn_sign_02_l [{shapeType=Box, extentX=4, extentY=2, extentZ=1.5, center=<2.1755 0.0001 0.0001>}]
 dyn_sign_02_r [{shapeType=Capsule, length=3.5, radius=1, shapeName=dyn_sign_02_r_capsule, center=<-1.9727 0 0>, rotation=<rot -90 0 0>}]
 dyn_sign_03_l [{shapeType=Capsule, length=3.5, radius=1.25, shapeName=dyn_sign_03_l_capsule, center=<3.6084 0 0.6169>, rotation=<rot -70.0005 180 -180>}]
 dyn_sign_03_r [{shapeType=Capsule, length=3.25, radius=1.25, shapeName=dyn_sign_03_r_capsule, center=<-2.7796 0 0>, rotation=<rot -90 0 0>}]
 dyn_sign_04_l [
  {shapeType=Capsule, length=4.25, radius=1, shapeName=dyn_sign_04_l_capsule, center=<2.0755 0 0>, rotation=<rot 0 89.9999 -89.999>}
  {shapeType=Box, extentX=4, extentY=5.5, extentZ=1.5, center=<2.1755 0.0001 0.0001>, rotation=<rot 0.0005 -0.0001 0>}
 ]
 dyn_sign_04_r [
  {shapeType=Capsule, length=4.25, radius=1, shapeName=dyn_sign_04_r_capsule, center=<-2.2123 0 0>, rotation=<rot 0 -90.0001 -90.0004>}
  {shapeType=Box, extentX=4, extentY=5.5, extentZ=1.5, center=<-2.1755 -0.0001 -0.0001>, rotation=<rot 0 0 -180>}
 ]
 dyn_sign_05_l [{shapeType=Capsule, length=4.5, radius=1.5, shapeName=dyn_sign_05_l_capsule, center=<3.9273 0 0>, rotation=<rot 0 89.9999 90.0004>}]
 dyn_sign_05_r [{shapeType=Capsule, length=5.25, radius=1, shapeName=dyn_sign_05_r_capsule, center=<-3.9923 0 -0.151>, rotation=<rot 90 0 0.0001>}]
 dyn_sign_06_l [
  {shapeType=Capsule, length=5, radius=1.25, shapeName=dyn_sign_06_l_capsule, center=<3.9602 0 0>, rotation=<rot 90 -179.9975 -179.9975>}
  {shapeType=Box, extentX=8, extentY=11.5, extentZ=1.5, center=<0.1 0 0>}
 ]
 dyn_sign_06_r [{shapeType=Box, extentX=8, extentY=11.5, extentZ=1.5, center=<-0.1 0 0>}]
 head [{shapeType=Capsule, length=3.75, radius=14.25, center=<9.6569 -0.0001 -0.0002>, rotation=<rot 0 0 -90.0001>}]
 neck_01 [{shapeType=Capsule, length=9.5, radius=11.75, rotation=<rot 90 0 0>}]
 pelvis [{shapeType=Capsule, length=17.1965, radius=10.1665}]
 spine_03 [
  {shapeType=Capsule, length=12.1965, radius=15.1665, center=<4.1385 -1.1124 0>, rotation=<rot 0 5.0002 0>}
  {shapeType=Capsule, length=24.75, radius=14, center=<11.6093 0.8271 -0.0002>, rotation=<rot 0 5.0003 0>}
  {shapeType=Box, extentX=70.5, extentY=63, extentZ=7.5, center=<9.9669 11.5188 19.9501>}
  {shapeType=Box, extentX=63.5, extentY=54.5, extentZ=7.5, center=<12.5997 15.149 -22.3893>}
 ]
 spine_04 [{shapeType=Capsule, length=15.4465, radius=14.9165, center=<6.9274 -1.0288 0>, rotation=<rot 0 -10.0004 0>}]
 spine_05 [{shapeType=Capsule, length=15.6965, radius=11.6665, center=<2.6159 -2.3149 0>}]
}
constraints [
 {
  bone1Name=neck_01
  bone2Name=spine_05
  swing1LimitDegrees=45
  swing1Motion=Free
  swing2LimitDegrees=45
  swing2Motion=Free
  twistLimitDegrees=45
  twistMotion=Free
 }
 {
  bone1Name=dyn_sign_01_l_MSTR
  bone2Name=spine_05
  swing1LimitDegrees=45
  swing1Motion=Free
  swing2LimitDegrees=45
  swing2Motion=Free
  twistLimitDegrees=45
  twistMotion=Free
 }
 {bone1Name=dyn_sign_01_l, bone2Name=dyn_sign_01_l_MSTR, swing1LimitDegrees=25, swing2LimitDegrees=25, twistLimitDegrees=5}
 {bone1Name=dyn_sign_02_l, bone2Name=dyn_sign_01_l, swing1LimitDegrees=25, swing2LimitDegrees=25, twistLimitDegrees=5}
 {bone1Name=dyn_sign_03_l, bone2Name=dyn_sign_02_l, swing1LimitDegrees=5, swing2LimitDegrees=45, twistLimitDegrees=5}
 {bone1Name=dyn_sign_04_l, bone2Name=dyn_sign_03_l, swing1LimitDegrees=45, swing2LimitDegrees=45, twistLimitDegrees=5}
 {bone1Name=dyn_sign_05_l, bone2Name=dyn_sign_04_l, swing1LimitDegrees=45, swing2LimitDegrees=45, twistLimitDegrees=5}
 {bone1Name=dyn_sign_06_l, bone2Name=dyn_sign_05_l, swing1LimitDegrees=45, swing2LimitDegrees=45, twistLimitDegrees=5}
 {
  bone1Name=dyn_sign_01_r_MSTR
  bone2Name=spine_05
  swing1LimitDegrees=45
  swing1Motion=Free
  swing2LimitDegrees=45
  swing2Motion=Free
  twistLimitDegrees=45
  twistMotion=Free
 }
 {
  bone1Name=spine_03
  bone2Name=pelvis
  swing1LimitDegrees=45
  swing1Motion=Free
  swing2LimitDegrees=45
  swing2Motion=Free
  twistLimitDegrees=45
  twistMotion=Free
 }
 {
  bone1Name=spine_04
  bone2Name=spine_03
  swing1LimitDegrees=45
  swing1Motion=Free
  swing2LimitDegrees=45
  swing2Motion=Free
  twistLimitDegrees=45
  twistMotion=Free
 }
 {
  bone1Name=spine_05
  bone2Name=spine_04
  swing1LimitDegrees=45
  swing1Motion=Free
  swing2LimitDegrees=45
  swing2Motion=Free
  twistLimitDegrees=45
  twistMotion=Free
 }
 {bone1Name=dyn_sign_01_r, bone2Name=dyn_sign_01_r_MSTR, swing1LimitDegrees=25, swing2LimitDegrees=25, twistLimitDegrees=5}
 {bone1Name=dyn_sign_02_r, bone2Name=dyn_sign_01_r, swing1LimitDegrees=25, swing2LimitDegrees=25, twistLimitDegrees=5}
 {bone1Name=dyn_sign_03_r, bone2Name=dyn_sign_02_r, swing1LimitDegrees=45, swing2LimitDegrees=45, twistLimitDegrees=5}
 {bone1Name=dyn_sign_04_r, bone2Name=dyn_sign_03_r, swing1LimitDegrees=45, swing2LimitDegrees=45, twistLimitDegrees=5}
 {bone1Name=dyn_sign_05_r, bone2Name=dyn_sign_04_r, swing1LimitDegrees=45, swing2LimitDegrees=45, twistLimitDegrees=5}
 {bone1Name=dyn_sign_06_r, bone2Name=dyn_sign_05_r, swing1LimitDegrees=45, swing2LimitDegrees=45, twistLimitDegrees=5}
 {
  bone1Name=dyn_sign_06_l
  bone2Name=dyn_sign_06_r
  swing1LimitDegrees=45
  swing1Motion=Free
  swing2LimitDegrees=45
  swing2Motion=Free
  twistLimitDegrees=45
  twistMotion=Free
 }
 {
  bone1Name=dyn_sign_05_l
  bone2Name=dyn_sign_05_r
  swing1LimitDegrees=45
  swing1Motion=Free
  swing2LimitDegrees=45
  swing2Motion=Free
  twistLimitDegrees=45
  twistMotion=Free
 }
 {
  bone1Name=dyn_sign_04_l
  bone2Name=dyn_sign_04_r
  swing1LimitDegrees=45
  swing1Motion=Free
  swing2LimitDegrees=45
  swing2Motion=Free
  twistLimitDegrees=45
  twistMotion=Free
 }
 {
  bone1Name=head
  bone2Name=neck_01
  swing1LimitDegrees=45
  swing1Motion=Free
  swing2LimitDegrees=45
  swing2Motion=Free
  twistLimitDegrees=45
  twistMotion=Free
 }
]
