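# Physics asset for the back straps of the Militia Burger mascot player cosmetic, defining bodies and joints for their dynamic bones.
# Left and right chains each run from a pelvis capsule through a tiny master box to two box-shaped strap segments.
# The master joints rotate freely against the pelvis, while segment joints have angular limits, with tight twist and wider, asymmetric swing.
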PhysicsAsset /BRCosmetics/Characters/Player/Male/Medium/Bodies/M_MED_Mascot_Militia_Burger/Meshes/M_MED_Mascot_Militia_Burger_straps_Physics
bodies {
 dyn_strap_back_01_l [{shapeType=Box, extentX=6, extentY=5, extentZ=1.5, center=<2.9863 0 -0.0001>}]
 dyn_strap_back_01_l_MSTR [{shapeType=Box, extentX=0.01, extentY=0.01, extentZ=0.01}]
 dyn_strap_back_01_r [{shapeType=Box, extentX=6, extentY=5, extentZ=1.5, center=<-2.9863 0 0.0001>, rotation=<rot 0 0 180>}]
 dyn_strap_back_01_r_MSTR [{shapeType=Box, extentX=0.01, extentY=0.01, extentZ=0.01}]
 dyn_strap_back_02_l [{shapeType=Box, extentX=6, extentY=5, extentZ=1.5, center=<2.9863 0 -0.0001>}]
 dyn_strap_back_02_r [{shapeType=Box, extentX=6, extentY=5, extentZ=1.5, center=<-2.9863 0 0.0001>, rotation=<rot 0 0 180>}]
 pelvis [{shapeType=Capsule, length=17.1965, radius=10.1665}]
}
constraints [
 {bone1Name=dyn_strap_back_01_l, bone2Name=dyn_strap_back_01_l_MSTR, swing1LimitDegrees=15, swing2LimitDegrees=45, twistLimitDegrees=5}
 {bone1Name=dyn_strap_back_02_l, bone2Name=dyn_strap_back_01_l, swing1LimitDegrees=15, swing2LimitDegrees=45, twistLimitDegrees=5}
 {
  bone1Name=dyn_strap_back_01_l_MSTR
  bone2Name=pelvis
  swing1LimitDegrees=45
  swing1Motion=Free
  swing2LimitDegrees=45
  swing2Motion=Free
  twistLimitDegrees=45
  twistMotion=Free
 }
 {
  bone1Name=dyn_strap_back_01_r_MSTR
  bone2Name=pelvis
  swing1LimitDegrees=45
  swing1Motion=Free
  swing2LimitDegrees=45
  swing2Motion=Free
  twistLimitDegrees=45
  twistMotion=Free
 }
 {bone1Name=dyn_strap_back_02_r, bone2Name=dyn_strap_back_01_r, swing1LimitDegrees=15, swing2LimitDegrees=45, twistLimitDegrees=5}
 {bone1Name=dyn_strap_back_01_r, bone2Name=dyn_strap_back_01_r_MSTR, swing1LimitDegrees=15, swing2LimitDegrees=45, twistLimitDegrees=5}
]
