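# Anim Blueprint for the Mascot Militia Burger cosmetic body, adding secondary motion to clothing and gear over the parent mesh's copied pose.
# Bone corrections and shoe flap drivers feed cached RigidBody simulations for cloth, straps and accessories, then knee, AnimDynamics leg-armor and grenade fixes.
# EventGraph derives alphas and hoodie rotations from head swing, arm angle, aiming, crouch and skydive state, plus skydiving wind forces.
# Crouch transitions reset dynamics.
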AnimBlueprint /BRCosmetics/Characters/Player/Male/Medium/Bodies/M_MED_Mascot_Militia_Burger/Meshes/M_MED_Mascot_Militia_Burger_AnimBP
target_skeleton=&/M_MED_Mascot_Militia_Burger_Skeleton
graphs [
 AnimGraph {nodes [
   Blend Poses by bool(bActiveValue=Get bIsInFrontEnd()) {blendType=Linear}
   n1 = "Constraint - dyn_armor_front_00_r (R)"(Alpha=1, ConstraintWeights_0=1) {
    alphaBoolBlend {blendOption=Linear}
    boneToModify {boneName=dyn_armor_front_00_r}
    constraintSetup [{transformType=Rotation, targetBone {boneName=thigh_r}}]
    constraintWeights [1]
   }
   "Constraint - dyn_armor_front_00_l (R)"(ComponentPose=@n1, Alpha=1, ConstraintWeights_0=1) {
    alphaBoolBlend {blendOption=Linear}
    boneToModify {boneName=dyn_armor_front_00_l}
    constraintSetup [{transformType=Rotation, targetBone {boneName=thigh_l}}]
    constraintWeights [1]
   }
   n2 = "Constraint - dyn_kneearmor_01_l (RR)"!(ComponentPose=Local To Component(LocalPose=Use cached pose 'simCache'()), Alpha=1, ConstraintWeights_0=0.5, ConstraintWeights_1:float=0.5) -> (Pose) {
    alphaBoolBlend {blendOption=Linear}
    boneToModify {boneName=dyn_kneearmor_01_l}
    constraintSetup [{transformType=Rotation, targetBone {boneName=calf_l}}, {transformType=Rotation, targetBone {boneName=thigh_l}}]
    constraintWeights [1,1]
   }
   n3 = "Constraint - dyn_kneearmor_01_r (RR)"!(ComponentPose=@n2, Alpha=1, ConstraintWeights_0=0.5, ConstraintWeights_1:float=0.5) -> (Pose) {
    alphaBoolBlend {blendOption=Linear}
    boneToModify {boneName=dyn_kneearmor_01_r}
    constraintSetup [{transformType=Rotation, targetBone {boneName=calf_r}}, {transformType=Rotation, targetBone {boneName=thigh_r}}]
    constraintWeights [1,1]
   }
   n4 = Get isCrouchingAlpha() -> (isCrouchingAlpha:double)
   n5 = "Transform (Modify) Bone - Bone: dyn_kneearmor_01_l"(ComponentPose=@n3, Alpha=@n4, Translation="-3,0,0", Rotation="0,0,0", Scale="1,1.2,1") {
    scaleMode=BMM_Additive
    scaleSpace=BCS_BoneSpace
    translationMode=BMM_Additive
    translationSpace=BCS_BoneSpace
    alphaBoolBlend {blendOption=Linear}
    boneToModify {boneName=dyn_kneearmor_01_l}
    scale {y=1.2}
    translation {x=-3}
   }
   n6 = "Transform (Modify) Bone - Bone: dyn_kneearmor_01_r"(ComponentPose=@n5, Alpha=@n4, Translation="3,0,0", Rotation="0,0,0", Scale="1,1.2,1") {
    scaleMode=BMM_Additive
    scaleSpace=BCS_BoneSpace
    translationMode=BMM_Additive
    translationSpace=BCS_BoneSpace
    alphaBoolBlend {blendOption=Linear}
    boneToModify {boneName=dyn_kneearmor_01_r}
    scale {y=1.2}
    translation {x=3}
   }
   n7 = Get windVector_1_5() -> (windVector_1_5:Vector)
   n8 = Get crouchingLegArmorAlpha() -> (crouchingLegArmorAlpha:double)
   n9 = Anim Dynamics - Bone: dyn_legarmor_01_r_MSTR(ComponentPose=@n6, Alpha=@n8, ExternalForce=@n7) {
    angularDampingOverride=1
    angularSpringConstant=50
    bAngularSpring=true
    bGravityOverrideInSimSpace=true
    alphaBoolBlend {blendOption=Linear}
    boundBone {boneName=dyn_legarmor_01_r_MSTR}
    componentAppliedLinearAccClamp=<1000 1000 1000>
    componentLinearAccScale=<0.3 0.3 0.6>
    componentLinearVelScale=<0.3 0.3 0.6>
    physicsBodyDefinitions [{
      sphereCollisionRadius=0
      boundBone {boneName=dyn_legarmor_01_r_MSTR}
      boxExtents=<35 35 35>
      constraintSetup {angularLimitsMax=<0 0 10>, angularLimitsMin=<0 -35 0>, angularTarget=<0 0 0>}
      localJointOffset=<-10 0 0>
     }]
    relativeSpaceBone {boneName=spine_01}
   }
   n10 = Anim Dynamics - Bone: dyn_legarmor_01_l_MSTR(ComponentPose=@n9, Alpha=@n8, ExternalForce=@n7) {
    angularDampingOverride=1
    angularSpringConstant=50
    bAngularSpring=true
    bGravityOverrideInSimSpace=true
    alphaBoolBlend {blendOption=Linear}
    boundBone {boneName=dyn_legarmor_01_l_MSTR}
    componentAppliedLinearAccClamp=<1000 1000 1000>
    componentLinearAccScale=<0.3 0.3 0.6>
    componentLinearVelScale=<0.3 0.3 0.6>
    gravityOverride {z=-980}
    physicsBodyDefinitions [{
      sphereCollisionRadius=0
      boundBone {boneName=dyn_legarmor_01_l_MSTR}
      boxExtents=<35 35 35>
      constraintSetup {angularLimitsMax=<0 0 10>, angularLimitsMin=<0 -35 0>, angularTarget=<0 0 0>}
      localJointOffset=<10 0 0>
     }]
    relativeSpaceBone {boneName=spine_01}
   }
   n11 = Get armSwingDeltaDown_r() -> (armSwingDeltaDown_r:double)
   n12 = "Constraint - dyn_grenade_r_01_MSTR (PP)"!(ComponentPose=@n10, Alpha=@n11, ConstraintWeights_0=0.5, ConstraintWeights_1:float=0.5) -> (Pose) {
    alphaBoolBlend {blendOption=Linear}
    boneToModify {boneName=dyn_grenade_r_01_MSTR}
    constraintSetup [{targetBone {boneName=spine_05}}, {targetBone {boneName=spine_04}}]
    constraintWeights [1,1]
   }
   n13 = "Transform (Modify) Bone - Bone: dyn_grenade_r_01_MSTR"(ComponentPose=@n12, Alpha=@n11, Translation="0,0,0", Rotation="-20,0,0", Scale="1,1,1") {
    rotationMode=BMM_Additive
    rotationSpace=BCS_BoneSpace
    translationMode=BMM_Additive
    alphaBoolBlend {blendOption=Linear}
    boneToModify {boneName=dyn_grenade_r_01_MSTR}
    rotation {pitch=-20}
   }
   n14 = "Transform (Modify) Bone - Bone: dyn_grenade_r_01_MSTR"(ComponentPose=@n13, Alpha=Get headAimingAlpha() -> (headAimingAlpha:double), Translation="0,0,3", Rotation="0,-45,20", Scale="1,1,1") {
    rotationMode=BMM_Additive
    rotationSpace=BCS_BoneSpace
    translationMode=BMM_Additive
    alphaBoolBlend {blendOption=Linear}
    boneToModify {boneName=dyn_grenade_r_01_MSTR}
    rotation {roll=20, yaw=-45}
    translation {z=3}
   }
   Output Pose(Result=Component To Local(ComponentPose=@n14)) {layerGroup=None}
   n15 = "Constraint - dyn_strap_back_01_r_MSTR (RR)"!(
    ComponentPose=Local To Component(LocalPose=Copy Pose From Mesh() {bUseAttachedParent=true})
    Alpha=1
    ConstraintWeights_0=0.5
    ConstraintWeights_1:float=0.5
   ) -> (Pose) {
    alphaBoolBlend {blendOption=Linear}
    boneToModify {boneName=dyn_strap_back_01_r_MSTR}
    constraintSetup [{transformType=Rotation, targetBone {boneName=spine_05}}, {transformType=Rotation, targetBone {boneName=clavicle_r}}]
    constraintWeights [1,1]
   }
   n16 = "Constraint - dyn_strap_back_01_l_MSTR (RR)"!(ComponentPose=@n15, Alpha=1, ConstraintWeights_0=0.5, ConstraintWeights_1:float=0.5) -> (Pose) {
    alphaBoolBlend {blendOption=Linear}
    boneToModify {boneName=dyn_strap_back_01_l_MSTR}
    constraintSetup [{transformType=Rotation, targetBone {boneName=spine_04}}, {transformType=Rotation, targetBone {boneName=clavicle_l}}]
    constraintWeights [1,1]
   }
   n17 = "Constraint - dyn_shoe_01_l (R)"(ComponentPose=@n16, Alpha=1, ConstraintWeights_0=1) {
    alphaBoolBlend {blendOption=Linear}
    boneToModify {boneName=dyn_shoe_01_l}
    constraintSetup [{transformType=Rotation, targetBone {boneName=calf_l}}]
    constraintWeights [1]
   }
   n18 = Flap Node - Bone: dyn_shoe_01_l!(ComponentPose:Component Space Pose Link Structure=@n17, Alpha:float=1) -> (Pose:Component Space Pose Link Structure) {
    is=title FortAnimGraphNode_Flap
    bUseClamp=true
    targetClampMax=1
    targetComponent=RotationY
    alphaBoolBlend {blendOption=Linear}
    sourceDrives [{
      bUseRange=false
      drivingCurve=null
      multiplier=1
      rangeMax=1
      rangeMin=-1
      remappedMax=1
      remappedMin=0
      sourceComponent=RotationZ
      useQuaternion=true
      sourceBone {boneName=foot_l}
     }]
    targetBone {boneName=dyn_shoe_01_l}
   }
   n19 = "Constraint - dyn_shoe_01_r (R)"(ComponentPose=@n18, Alpha=1, ConstraintWeights_0=1) {
    alphaBoolBlend {blendOption=Linear}
    boneToModify {boneName=dyn_shoe_01_r}
    constraintSetup [{transformType=Rotation, targetBone {boneName=calf_r}}]
    constraintWeights [1]
   }
   n20 = Flap Node - Bone: dyn_shoe_01_r!(ComponentPose:Component Space Pose Link Structure=@n19, Alpha:float=1) -> (Pose:Component Space Pose Link Structure) {
    is=title FortAnimGraphNode_Flap
    bUseClamp=true
    targetClampMax=1
    targetComponent=RotationY
    alphaBoolBlend {blendOption=Linear}
    sourceDrives [{
      bUseRange=false
      drivingCurve=null
      multiplier=1
      rangeMax=1
      rangeMin=-1
      remappedMax=1
      remappedMin=0
      sourceComponent=RotationZ
      useQuaternion=true
      sourceBone {boneName=foot_r}
     }]
    targetBone {boneName=dyn_shoe_01_r}
   }
   n21 = Get headSwingAngleSkydiveAlpha() -> (headSwingAngleSkydiveAlpha:double)
   n22 = "Transform (Modify) Bone - Bone: dyn_hoodie_back_01_MSTR"(ComponentPose=@n20, Alpha=@n21, Translation="0,0,0", Rotation=Get hoodieRotator_c() -> (hoodieRotator_c:Rotator), Scale="1,1,1") {rotationMode=BMM_Additive, rotationSpace=BCS_BoneSpace, alphaBoolBlend {blendOption=Linear}, boneToModify {boneName=dyn_hoodie_back_01_MSTR}}
   n23 = "Transform (Modify) Bone - Bone: dyn_hoodie_01_r_MSTR"(ComponentPose=@n22, Alpha=@n21, Translation="0,0,0", Rotation="90,0,0", Scale="1,1,1") {
    rotationMode=BMM_Additive
    rotationSpace=BCS_BoneSpace
    alphaBoolBlend {blendOption=Linear}
    boneToModify {boneName=dyn_hoodie_01_r_MSTR}
    rotation {pitch=90}
   }
   n24 = "Transform (Modify) Bone - Bone: dyn_hoodie_back_01_l_MSTR"(ComponentPose=@n23, Alpha=@n21, Translation="0,0,0", Rotation=Get hoodieRotator_l() -> (hoodieRotator_l:Rotator), Scale="1,1,1") {
    rotationMode=BMM_Additive
    rotationSpace=BCS_BoneSpace
    alphaBoolBlend {blendOption=Linear}
    boneToModify {boneName=dyn_hoodie_back_01_l_MSTR}
   }
   n25 = "Transform (Modify) Bone - Bone: dyn_hoodie_back_01_r_MSTR"(
    ComponentPose=@n24
    Alpha=Get headSwingAngleAlpha() -> (headSwingAngleAlpha:double)
    Translation="0,0,0"
    Rotation=Get hoodieRotator_r() -> (hoodieRotator_r:Rotator)
    Scale="1,1,1"
   ) {
    rotationMode=BMM_Additive
    rotationSpace=BCS_BoneSpace
    alphaBoolBlend {blendOption=Linear}
    boneToModify {boneName=dyn_hoodie_back_01_r_MSTR}
   }
   n26 = "Transform (Modify) Bone - Bone: dyn_armor_front_01_MSTR"(ComponentPose=@n25, Alpha=Get isCrouchingAlpha(), Translation="-4,4,0", Rotation="-10,0,-35", Scale="1,1,1") {
    rotationMode=BMM_Additive
    translationMode=BMM_Additive
    alphaBoolBlend {blendOption=Linear}
    boneToModify {boneName=dyn_armor_front_01_MSTR}
    rotation {pitch=-10, roll=-35}
    translation=<-4 4>
   }
   n27 = "Transform (Modify) Bone - Bone: dyn_sign_01_l_MSTR"(ComponentPose=@n26, Alpha=1, Translation="-2.5,0,1", Rotation="0,0,0", Scale="1,1,1") {
    rotationMode=BMM_Additive
    rotationSpace=BCS_BoneSpace
    translationMode=BMM_Additive
    translationSpace=BCS_BoneSpace
    alphaBoolBlend {blendOption=Linear}
    boneToModify {boneName=dyn_sign_01_l_MSTR}
    translation {x=-2.5, z=1}
   }
   n28 = "Transform (Modify) Bone - Bone: dyn_sign_01_r_MSTR"(ComponentPose=@n27, Alpha=1, Translation="2.5,0,-1", Rotation="0,0,0", Scale="1,1,1") {
    rotationMode=BMM_Additive
    rotationSpace=BCS_BoneSpace
    translationMode=BMM_Additive
    translationSpace=BCS_BoneSpace
    alphaBoolBlend {blendOption=Linear}
    boneToModify {boneName=dyn_sign_01_r_MSTR}
    translation {x=2.5, z=-1}
   }
   n29 = "Transform (Modify) Bone - Bone: dyn_sign_03_l"(ComponentPose=@n28, Alpha=1, Translation="0,0,0", Rotation="15,0,0", Scale="1,1,1") {
    rotationMode=BMM_Additive
    rotationSpace=BCS_BoneSpace
    translationMode=BMM_Additive
    translationSpace=BCS_BoneSpace
    alphaBoolBlend {blendOption=Linear}
    boneToModify {boneName=dyn_sign_03_l}
    rotation {pitch=15}
   }
   n30 = "Transform (Modify) Bone - Bone: dyn_sign_03_r"(ComponentPose=@n29, Alpha=1, Translation="0,0,0", Rotation="15,0,0", Scale="1,1,1") {
    rotationMode=BMM_Additive
    rotationSpace=BCS_BoneSpace
    translationMode=BMM_Additive
    translationSpace=BCS_BoneSpace
    alphaBoolBlend {blendOption=Linear}
    boneToModify {boneName=dyn_sign_03_r}
    rotation {pitch=15}
   }
   n31 = Get headAimingAlpha()
   n32 = "Constraint - dyn_sign_01_l_MSTR (PP)"!(ComponentPose=@n30, Alpha=@n31, ConstraintWeights_0=0.5, ConstraintWeights_1:float=0.5) -> (Pose) {
    alphaBoolBlend {blendOption=Linear}
    boneToModify {boneName=dyn_sign_01_l_MSTR}
    constraintSetup [{targetBone {boneName=neck_02}}, {targetBone {boneName=spine_05}}]
    constraintWeights [1,1]
   }
   n33 = "Constraint - dyn_sign_01_r_MSTR (P)"(ComponentPose=@n32, Alpha=@n31, ConstraintWeights_0=1) {
    alphaBoolBlend {blendOption=Linear}
    boneToModify {boneName=dyn_sign_01_r_MSTR}
    constraintSetup [{targetBone {boneName=neck_02}}]
    constraintWeights [1]
   }
   n34 = "Constraint - dyn_pouch_front_l_03_MSTR (PP)"!(ComponentPose=@n33, Alpha=1, ConstraintWeights_0=0.5, ConstraintWeights_1:float=0.5) -> (Pose) {
    alphaBoolBlend {blendOption=Linear}
    boneToModify {boneName=dyn_pouch_front_l_03_MSTR}
    constraintSetup [{targetBone {boneName=spine_04}}, {targetBone {boneName=dyn_pouch_front_l_02}}]
    constraintWeights [1,1]
   }
   n35 = "Transform (Modify) Bone - Bone: dyn_pouch_front_l_03_MSTR"(ComponentPose=@n34, Alpha=1, Translation="0,1.8,0", Rotation="0,0,0", Scale="1,1,1") {translationMode=BMM_Additive, alphaBoolBlend {blendOption=Linear}, boneToModify {boneName=dyn_pouch_front_l_03_MSTR}, translation {y=1.8}}
   n36 = "Constraint - dyn_pouch_front_03_MSTR (PP)"!(ComponentPose=@n35, Alpha=1, ConstraintWeights_0=0.5, ConstraintWeights_1:float=0.5) -> (Pose) {
    alphaBoolBlend {blendOption=Linear}
    boneToModify {boneName=dyn_pouch_front_03_MSTR}
    constraintSetup [{targetBone {boneName=spine_04}}, {targetBone {boneName=dyn_pouch_front_02}}]
    constraintWeights [1,1]
   }
   n37 = "Constraint - dyn_pouch_front_r_03_MSTR (PP)"!(ComponentPose=@n36, Alpha=1, ConstraintWeights_0=0.5, ConstraintWeights_1:float=0.5) -> (Pose) {
    alphaBoolBlend {blendOption=Linear}
    boneToModify {boneName=dyn_pouch_front_r_03_MSTR}
    constraintSetup [{targetBone {boneName=spine_04}}, {targetBone {boneName=dyn_pouch_front_r_02}}]
    constraintWeights [1,1]
   }
   Save cached pose 'preSimCache'(Pose=Component To Local(ComponentPose=@n37)) {cacheName=preSimCache}
   n38 = Get windVector_0_75() -> (windVector_0_75:Vector)
   n39 = RigidBody(ComponentPose=Local To Component(LocalPose=Use cached pose 'preSimCache'()), ExternalForce=@n38) {
    bClampLinearTranslationLimitToRefPose=true
    bForceDisableCollisionBetweenConstraintBodies=true
    bOverrideWorldGravity=true
    overridePhysicsAsset=&/M_MED_Mascot_Militia_Burger_loincloth_Physics
    simulationSpace=BaseBoneSpace
    alphaBoolBlend {blendOption=Linear}
    baseBoneRef {boneName=pelvis}
    componentAppliedLinearAccClamp=<3000 3000 3000>
    componentLinearAccScale=<0.3 0.3 0.6>
    componentLinearVelScale=<0.3 0.3 0.6>
    overrideWorldGravity {z=-980}
   }
   n40 = RigidBody(ComponentPose=@n39, ExternalForce=@n38) {
    bClampLinearTranslationLimitToRefPose=true
    bForceDisableCollisionBetweenConstraintBodies=true
    bOverrideWorldGravity=true
    overridePhysicsAsset=&/M_MED_Mascot_Militia_Burger_Physics
    simulationSpace=BaseBoneSpace
    alphaBoolBlend {blendOption=Linear}
    baseBoneRef {boneName=spine_01}
    componentAppliedLinearAccClamp=<3000 3000 3000>
    componentLinearAccScale=<0.3 0.3 0.6>
    componentLinearVelScale=<0.3 0.3 0.6>
    overrideWorldGravity {z=-980}
   }
   n41 = RigidBody(ComponentPose=@n40, Alpha=Get hoodieRBDOnOffAlpha() -> (hoodieRBDOnOffAlpha:double), ExternalForce=@n38) {
    bClampLinearTranslationLimitToRefPose=true
    bForceDisableCollisionBetweenConstraintBodies=true
    bOverrideWorldGravity=true
    overridePhysicsAsset=&/M_MED_Mascot_Militia_Burger_hoodie_Physics
    simulationSpace=BaseBoneSpace
    alphaBoolBlend {blendOption=Linear}
    baseBoneRef {boneName=spine_01}
    componentAppliedLinearAccClamp=<3000 3000 3000>
    componentLinearAccScale=<0.3 0.3 0.6>
    componentLinearVelScale=<0.3 0.3 0.6>
    overrideWorldGravity {z=-980}
   }
   n42 = RigidBody(ComponentPose=@n41, ExternalForce=@n38) {
    bClampLinearTranslationLimitToRefPose=true
    bForceDisableCollisionBetweenConstraintBodies=true
    bOverrideWorldGravity=true
    overridePhysicsAsset=&/M_MED_Mascot_Militia_Burger_necklace_Physics
    simulationSpace=BaseBoneSpace
    alphaBoolBlend {blendOption=Linear}
    baseBoneRef {boneName=spine_01}
    componentAppliedLinearAccClamp=<3000 3000 3000>
    componentLinearAccScale=<0.3 0.3 0.6>
    componentLinearVelScale=<0.3 0.3 0.6>
    overrideWorldGravity {z=-980}
   }
   n43 = Get windVector_1_5()
   n44 = RigidBody(ComponentPose=@n42, ExternalForce=@n43) {
    bClampLinearTranslationLimitToRefPose=true
    bForceDisableCollisionBetweenConstraintBodies=true
    bOverrideWorldGravity=true
    overridePhysicsAsset=&/M_MED_Mascot_Militia_Burger_leg_l_Physics
    simulationSpace=BaseBoneSpace
    alphaBoolBlend {blendOption=Linear}
    baseBoneRef {boneName=calf_l}
    componentAppliedLinearAccClamp=<3000 3000 3000>
    componentLinearAccScale=<0.3 0.3 0.6>
    componentLinearVelScale=<0.3 0.3 0.6>
    overrideWorldGravity {z=-980}
   }
   n45 = RigidBody(ComponentPose=@n44, ExternalForce=@n43) {
    bClampLinearTranslationLimitToRefPose=true
    bForceDisableCollisionBetweenConstraintBodies=true
    bOverrideWorldGravity=true
    overridePhysicsAsset=&/M_MED_Mascot_Militia_Burger_leg_r_Physics
    simulationSpace=BaseBoneSpace
    alphaBoolBlend {blendOption=Linear}
    baseBoneRef {boneName=calf_r}
    componentAppliedLinearAccClamp=<3000 3000 3000>
    componentLinearAccScale=<0.3 0.3 0.6>
    componentLinearVelScale=<0.3 0.3 0.6>
    overrideWorldGravity {z=-980}
   }
   n46 = RigidBody(ComponentPose=@n45, ExternalForce=@n43) {
    bClampLinearTranslationLimitToRefPose=true
    bForceDisableCollisionBetweenConstraintBodies=true
    bOverrideWorldGravity=true
    overridePhysicsAsset=&/M_MED_Mascot_Militia_Burger_straps_Physics
    simulationSpace=BaseBoneSpace
    alphaBoolBlend {blendOption=Linear}
    baseBoneRef {boneName=spine_05}
    componentAppliedLinearAccClamp=<3000 3000 3000>
    componentLinearAccScale=<0.3 0.3 0.6>
    componentLinearVelScale=<0.3 0.3 0.6>
    overrideWorldGravity {z=-980}
   }
   Save cached pose 'simCache'(Pose=Component To Local(ComponentPose=@n46)) {cacheName=simCache}
  ]}
 EventGraph {nodes [
   Event BlueprintUpdateAnimation()
   > n1 = Sequence!(execute) -> (then_0, then_1, then_2:Exec, then_3:Exec)
   n2 = GetDeltaTransformFromRefPose(self=GetOwnerSkeletalMeshComponent(), BoneName=upperarm_l, BaseName=clavicle_l)
   n3 = "Absolute (Float)"(A=BreakRotator(InRot=BreakTransform(InTransform=@n2).Rotation).Yaw)
   n4 = "To Float (Boolean)"(InBool=Get bIsTargeting())
   Set armSwingDeltaDown_r(execute=@n1.then_3, armSwingDeltaDown_r:double=MapRangeClamped(Value=@n3, InRangeA=20, InRangeB=70, OutRangeA=@n4, OutRangeB=0)) -> (Output_Get:double)
   > Set GrenadeSlideAlpha(GrenadeSlideAlpha:double=MapRangeClamped(Value=@n3, InRangeA=30, InRangeB=40, OutRangeA=@n4, OutRangeB=0)) -> (Output_Get:double)
   n5 = GetTwistAndSwingAngleOfDeltaRotationFromRefPose(self=GetOwnerSkeletalMeshComponent(), BoneName=head)
   > Set headAimingAlpha(headAimingAlpha:double=SelectFloat(A=MapRangeClamped(Value=@n5.OutSwingAngle, InRangeA=5, InRangeB=50), B=0, bPickA=Get bIsTargeting())) -> (Output_Get:double)
   n6 = Break Fort Anim Input Skydiving External Force(FortAnimInput_SkydivingExternalForce=Get SkydivingExternalForceInput()) -> (
    bUseSkydivingVectorForce=~
    SkydivingVectorForce=~
    HeadToPelvisDirection=~
    FloatingMultiplier=~
    FloatingAdditive=~
    DivingMultiplier=~
    DivingAdditive=~
    ParachutingMultiplier=~
    ParachutingAdditive=~
   )
   Set windVector_1_5(execute=@n1.then_0, windVector_1_5:Vector=vector * float(A=@n6.SkydivingVectorForce, B=1.5)) -> (Output_Get:Vector)
   > Set windVector_0_75(windVector_0_75:Vector=vector * float(A=@n6.SkydivingVectorForce, B=0.75)) -> (Output_Get:Vector)
   n7 = Get bIsSkydiveFloating()
   Set isSkydiveFloatingAlpha(execute=@n1.then_2, isSkydiveFloatingAlpha:double="To Float (Boolean)"(InBool=@n7)) -> (Output_Get:double)
   n8 = NOT Boolean(A=@n7)
   n9 = NOT Boolean(A=Get bIsSkydiveDiveMode())
   n10 = Get bIsSkyDiving()
   > Set isSkydiveNotFloating(isSkydiveNotFloating:double="To Float (Boolean)"(InBool=AND Boolean!(A=@n8, B=@n10, C:bool=@n9) -> (ReturnValue))) -> (Output_Get:double)
   n11 = Get bIsSkyDiving()
   > Set isNotSkydiving(isNotSkydiving:double="To Float (Boolean)"(InBool=NOT Boolean(A=@n11))) -> (Output_Get:double)
   n12 = GetTwistAndSwingAngleOfDeltaRotationFromRefPose(self=GetOwnerSkeletalMeshComponent(), BoneName=head)
   Set headSwingAngleAlpha(execute=@n1.then_1, headSwingAngleAlpha:double=MapRangeClamped(Value=@n12.OutSwingAngle, InRangeA=10, InRangeB=45)) -> (Output_Get:double)
   n13 = Get bIsSkydiveFloating()
   > Set headSwingAngleSkydiveAlpha(headSwingAngleSkydiveAlpha:double=MapRangeClamped(Value=@n12.OutSwingAngle, InRangeA=10, InRangeB=45, OutRangeB=SelectFloat(A=7, B=1, bPickA=@n13))) -> (Output_Get:double)
   > Set hoodieRotator_l(hoodieRotator_l:Rotator=MakeRotator(Pitch=SelectFloat(A=75, B=20, bPickA=@n13))) -> (Output_Get:Rotator)
   n14 = Get bIsCrouching()
   > Set hoodieRotator_r(hoodieRotator_r:Rotator=MakeRotator(Pitch=SelectFloat(A=135, B=90, bPickA=@n14))) -> (Output_Get:Rotator)
   n15 = "To Float (Boolean)"(InBool=@n14)
   n16 = Get bIsSkyDiving()
   > Set hoodieRBDOnOffAlpha(hoodieRBDOnOffAlpha:double=SelectFloat(A=MapRangeClamped(Value=@n12.OutSwingAngle, InRangeA=40, InRangeB=10, OutRangeB=@n15), B=SelectFloat(A=0.5, B=1, bPickA=@n16), bPickA=@n14)) -> (Output_Get:double)
   > Set hoodieRotator_c(hoodieRotator_c:Rotator=MakeRotator(Pitch=SelectFloat(A=110, B=55, bPickA=@n14))) -> (Output_Get:Rotator)
   Event BlueprintInitializeAnimation()
   > Set isCrouchingAlpha(isCrouchingAlpha:double=0) -> (Output_Get:double)
   > Set isNotCrouchingAlpha(isNotCrouchingAlpha:double=1) -> (Output_Get:double)
   > Set isCrouchingWeightAlpha_0(isCrouchingWeightAlpha_0:double=0.5) -> (Output_Get:double)
   > Set crouchingLegArmorAlpha(crouchingLegArmorAlpha:double=0.5) -> (Output_Get:double)
   > Set crouchingLegArmorAlpha(crouchingLegArmorAlpha=1)
   > ResetDynamics(InTeleportType=TeleportPhysics)
   > ResetDynamics(InTeleportType=ResetPhysics)
   Event OnBeginCrouching()
   > Set isCrouchingAlpha(isCrouchingAlpha=1)
   > Set isNotCrouchingAlpha(isNotCrouchingAlpha=0)
   > Set isCrouchingWeightAlpha_0(isCrouchingWeightAlpha_0=1)
   > Set crouchingLegArmorAlpha(crouchingLegArmorAlpha=0)
   > Set crouchingLegArmorAlpha(crouchingLegArmorAlpha=0.25)
   > ResetDynamics(InTeleportType=TeleportPhysics)
   > ResetDynamics(InTeleportType=ResetPhysics)
   Event OnEndCrouching()
   > Set isCrouchingAlpha(isCrouchingAlpha=0)
   > Set isNotCrouchingAlpha(isNotCrouchingAlpha=1)
   > Set isCrouchingWeightAlpha_0(isCrouchingWeightAlpha_0=0.5)
   > Set crouchingLegArmorAlpha(crouchingLegArmorAlpha=0.5)
   > Set crouchingLegArmorAlpha(crouchingLegArmorAlpha=1)
   > Delay(Duration=0.125)
   > ResetDynamics(InTeleportType=TeleportPhysics)
   > ResetDynamics(InTeleportType=ResetPhysics)
   TryGetPawnOwner()
   AND Boolean!(A=false, B=false) -> (ReturnValue)
   Get bIsTargeting()
  ]}
]
variables [
 armSwingDeltaDown_r {type=number}
 crouchingLegArmorAlpha {type=number}
 FrontEnd Alpha {type=number}
 GrenadeSlideAlpha {type=number}
 headAimingAlpha {type=number}
 headSwingAngleAlpha {type=number}
 headSwingAngleSkydiveAlpha {type=number}
 hoodieRBDOnOffAlpha {type=number}
 hoodieRotator_c {type=Rotator}
 hoodieRotator_l {type=Rotator}
 hoodieRotator_r {type=Rotator, default=<rot 90 0 0>}
 isCrouchingAlpha {type=number}
 isCrouchingWeightAlpha_0 {type=number}
 isNotCrouchingAlpha {type=number}
 isNotSkydiving {type=number}
 isSkydiveFloatingAlpha {type=number}
 isSkydiveNotFloating {type=number}
 windVector_0_75 {type=Vector}
 windVector_1_5 {type=Vector}
]
