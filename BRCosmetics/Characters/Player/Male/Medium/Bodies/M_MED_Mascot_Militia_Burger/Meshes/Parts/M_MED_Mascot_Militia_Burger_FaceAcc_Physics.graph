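# Physics asset for the face accessory mesh of the male Mascot Militia Burger character body, setting up jiggle for its cheese pieces and tongue.
# Box bodies sit on the back cheese, two front cheese and two tongue bones, with tiny boxes on MSTR bones acting as anchors.
# Each cheese piece is constrained to its own anchor, while the tongue forms a two-link chain from its anchor.
# Angular limits mostly favor one swing axis and keep twist tight.
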PhysicsAsset /BRCosmetics/Characters/Player/Male/Medium/Bodies/M_MED_Mascot_Militia_Burger/Meshes/Parts/M_MED_Mascot_Militia_Burger_FaceAcc_Physics
bodies {
 dyn_backcheese_01 [{shapeType=Box, extentX=3, extentY=6.5, extentZ=2, center=<1.4536 0 0>}]
 dyn_backcheese_01_MSTR [{shapeType=Box, extentX=0.01, extentY=0.01, extentZ=0.01}]
 dyn_frontcheeseB_01 [{shapeType=Box, extentX=3, extentY=6.5, extentZ=2, center=<1.4536 0 0>}]
 dyn_frontcheeseB_01_MSTR [{shapeType=Box, extentX=0.01, extentY=0.01, extentZ=0.01}]
 dyn_frontcheese_01 [{shapeType=Box, extentX=3, extentY=6.5, extentZ=2, center=<1.4536 0 0>}]
 dyn_frontcheese_01_MSTR [{shapeType=Box, extentX=0.01, extentY=0.01, extentZ=0.01}]
 dyn_tongue_02 [{shapeType=Box, extentX=8.5, extentY=10.5, extentZ=5, center=<5.1764 0.0001 0>}]
 dyn_tongue_02_MSTR [{shapeType=Box, extentX=0.01, extentY=0.01, extentZ=0.01}]
 dyn_tongue_03 [{shapeType=Box, extentX=8.5, extentY=10.5, extentZ=5, center=<4.5037 0.0001 0>}]
}
constraints [
 {bone1Name=dyn_backcheese_01, bone2Name=dyn_backcheese_01_MSTR, swing1LimitDegrees=5, swing2LimitDegrees=45, twistLimitDegrees=5}
 {bone1Name=dyn_frontcheese_01, bone2Name=dyn_frontcheese_01_MSTR, swing1LimitDegrees=15, swing2LimitDegrees=45, twistLimitDegrees=5}
 {bone1Name=dyn_frontcheeseB_01, bone2Name=dyn_frontcheeseB_01_MSTR, swing1LimitDegrees=15, swing2LimitDegrees=45, twistLimitDegrees=5}
 {bone1Name=dyn_tongue_02, bone2Name=dyn_tongue_02_MSTR, swing1LimitDegrees=45, swing2LimitDegrees=5, twistLimitDegrees=5}
 {bone1Name=dyn_tongue_03, bone2Name=dyn_tongue_02, swing1LimitDegrees=45, swing2LimitDegrees=45, twistLimitDegrees=5}
]
